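# Physics asset for the swinging medallion on the BoldDormRise medium male body cosmetic.
# Capsule bodies form a chain from dyn_medallion_base through three medallion bones, with twist limited to zero and swing mostly on one axis.
# The base joins spine_05 with free angular motion, as does spine_05 to spine_04.
# Larger capsules on both spine bones approximate the torso, likely as collision for the medallion.
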PhysicsAsset /BRCosmetics/Characters/Player/Male/Medium/Bodies/M_MED_BoldDormRise/Meshes/M_MED_BoldDormRise_Medallion_Physics
bodies {
 dyn_medallion_01 [{shapeType=Capsule, length=3, radius=2, shapeName=dyn_medallion_01_capsule, center=<2.8467 0 0>, rotation=<rot 0 0 90>}]
 dyn_medallion_02 [{shapeType=Capsule, length=3, radius=2, shapeName=dyn_medallion_02_capsule, center=<1.3495 0 0>, rotation=<rot 0 0 90>}]
 dyn_medallion_03 [{shapeType=Capsule, length=3, radius=2.2, shapeName=dyn_medallion_03_capsule, center=<4.1731 0 0>, rotation=<rot 90 0 0>}]
 dyn_medallion_base [{shapeType=Capsule, length=0.505, radius=0.505, shapeName=dyn_medallion_base_capsule, rotation=<rot 0 -90 0>}]
 spine_04 [
  {shapeType=Capsule, length=13.9328, radius=13, shapeName=spine_04_capsule, center=<-0.3782 -1.1867 0>, rotation=<rot 90 -15.8699 0>}
  {shapeType=Capsule, length=6, radius=4, shapeName=spine_04_capsule, center=<7.2474 11.3177 8.939>, rotation=<rot 86.1073 -15.87 -5.6917>}
 ]
 spine_05 [{shapeType=Capsule, length=10, radius=9, shapeName=spine_05_capsule, center=<-1.1533 0 0>}]
}
constraints [
 {bone1Name=dyn_medallion_01, bone2Name=dyn_medallion_base, swing1LimitDegrees=0, swing2LimitDegrees=40, twistLimitDegrees=0}
 {bone1Name=dyn_medallion_02, bone2Name=dyn_medallion_01, swing1LimitDegrees=5, swing2LimitDegrees=40, twistLimitDegrees=0}
 {bone1Name=dyn_medallion_03, bone2Name=dyn_medallion_02, swing1LimitDegrees=5, swing2LimitDegrees=40, twistLimitDegrees=0}
 {
  bone1Name=dyn_medallion_base
  bone2Name=spine_05
  swing1LimitDegrees=45
  swing1Motion=Free
  swing2LimitDegrees=45
  swing2Motion=Free
  twistLimitDegrees=45
  twistMotion=Free
 }
 {
  bone1Name=spine_05
  bone2Name=spine_04
  swing1LimitDegrees=45
  swing1Motion=Free
  swing2LimitDegrees=45
  swing2Motion=Free
  twistLimitDegrees=45
  twistMotion=Free
 }
]
