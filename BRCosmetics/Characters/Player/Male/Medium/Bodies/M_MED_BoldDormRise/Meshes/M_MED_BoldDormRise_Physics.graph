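# Physics asset for the BoldDormRise medium male player body, setting up capsule bodies for dynamic motion of the jacket's lower section.
# Seven front, side and back bone chains each run from a tiny base capsule through two larger segments with limited swing and twist.
# Bases, spine and thighs join the pelvis with free angular motion, giving body volumes around the jacket.
# An extra angularly free constraint links the left front and side tips.
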PhysicsAsset /BRCosmetics/Characters/Player/Male/Medium/Bodies/M_MED_BoldDormRise/Meshes/M_MED_BoldDormRise_Physics
bodies {
 dyn_jacket_bk_01 [{shapeType=Capsule, length=5, radius=3, shapeName=dyn_jacket_bk_01_capsule, center=<6.3341 0 -2.0729>, rotation=<rot 0 0 90>}]
 dyn_jacket_bk_01_l [{shapeType=Capsule, length=5, radius=3, shapeName=dyn_jacket_bk_01_l_capsule, center=<6.3341 0 -1.6033>, rotation=<rot 0 0 90>}]
 dyn_jacket_bk_01_r [{shapeType=Capsule, length=5, radius=3, shapeName=dyn_jacket_bk_01_r_capsule, center=<6.3341 0 -1.3419>, rotation=<rot 0 0 90>}]
 dyn_jacket_bk_02 [{shapeType=Capsule, length=7, radius=3.3, shapeName=dyn_jacket_bk_02_capsule, center=<4 0 -1.3478>, rotation=<rot 0 0 90>}]
 dyn_jacket_bk_02_l [{shapeType=Capsule, length=7, radius=3.3, shapeName=dyn_jacket_bk_02_l_capsule, center=<4 -0.0383 -1.4205>, rotation=<rot 0 0 84.242>}]
 dyn_jacket_bk_02_r [{shapeType=Capsule, length=7, radius=3, shapeName=dyn_jacket_bk_02_r_capsule, center=<4 0.0115 -1.3527>, rotation=<rot 0 0 92.9083>}]
 dyn_jacket_bk_base [{shapeType=Capsule, length=0.505, radius=0.505, shapeName=dyn_jacket_bk_base_capsule, rotation=<rot 0 -90 0>}]
 dyn_jacket_bk_base_l [{shapeType=Capsule, length=0.505, radius=0.505, shapeName=dyn_jacket_bk_base_l_capsule, rotation=<rot 0 -90 0>}]
 dyn_jacket_bk_base_r [{shapeType=Capsule, length=0.505, radius=0.505, shapeName=dyn_jacket_bk_base_r_capsule, rotation=<rot 0 -90 0>}]
 dyn_jacket_fr_01_l [{shapeType=Capsule, length=7, radius=3, shapeName=dyn_jacket_fr_01_l_capsule, center=<7.5818 0 -1.327>, rotation=<rot 0 0 90>}]
 dyn_jacket_fr_01_r [{shapeType=Capsule, length=5, radius=3, shapeName=dyn_jacket_fr_01_r_capsule, center=<6.777 0 -1.24>, rotation=<rot 0 0 90>}]
 dyn_jacket_fr_02_l [
  {shapeType=Capsule, length=9, radius=3.3, shapeName=dyn_jacket_fr_02_l_capsule, center=<4.5262 0 -1.8389>, rotation=<rot -0.0001 -0.0001 90>}
  {
   shapeType=Capsule
   length=2
   radius=2
   shapeName=dyn_jacket_fr_02_l_capsule
   center=<5.4405 -3.1968 0.4907>
   rotation=<rot -0.0002 -0.0002 123.8055>
  }
 ]
 dyn_jacket_fr_02_r [
  {shapeType=Capsule, length=9, radius=3.3, shapeName=dyn_jacket_fr_02_r_capsule, center=<4.5262 0 -1.8389>, rotation=<rot -0.0003 0.0002 90>}
  {shapeType=Capsule, length=2, radius=2, shapeName=dyn_jacket_fr_02_r_capsule, center=<5.128 3.3784 0.2468>, rotation=<rot 0 0.0001 -119.8833>}
 ]
 dyn_jacket_fr_base_l [{shapeType=Capsule, length=0.505, radius=0.505, shapeName=dyn_jacket_fr_base_l_capsule, rotation=<rot 0 -90 0>}]
 dyn_jacket_fr_base_r [{shapeType=Capsule, length=0.505, radius=0.505, shapeName=dyn_jacket_fr_base_r_capsule, rotation=<rot 0 -90 0>}]
 dyn_jacket_side_01_l [{shapeType=Capsule, length=5, radius=3.3, shapeName=dyn_jacket_side_01_l_capsule, center=<6.3 0 -1.6>, rotation=<rot 0 0 90>}]
 dyn_jacket_side_01_r [{shapeType=Capsule, length=5, radius=3.3, shapeName=dyn_jacket_side_01_r_capsule, center=<6.3 0 -1.6>, rotation=<rot 0 0 90>}]
 dyn_jacket_side_02_l [{shapeType=Capsule, length=8, radius=3, shapeName=dyn_jacket_side_02_l_capsule, center=<4.6 0 -2>, rotation=<rot 0 0 90>}]
 dyn_jacket_side_02_r [{shapeType=Capsule, length=5, radius=3, shapeName=dyn_jacket_side_02_r_capsule, center=<4.6 0 -2>, rotation=<rot 0 0 90>}]
 dyn_jacket_side_base_l [{shapeType=Capsule, length=0.505, radius=0.505, shapeName=dyn_jacket_side_base_l_capsule, rotation=<rot 0 -90 0>}]
 dyn_jacket_side_base_r [{shapeType=Capsule, length=0.505, radius=0.505, shapeName=dyn_jacket_side_base_r_capsule, rotation=<rot 0 -90 0>}]
 pelvis [{shapeType=Capsule, length=9, radius=10.5, shapeName=pelvis_capsule, center=<1.2394 -0.568 0.0758>, rotation=<rot -0.2933 -90 -0.3032>}]
 spine_01 [{shapeType=Capsule, length=12, radius=12.3, shapeName=spine_01_capsule, center=<8 -0.6622 -0.3107>, rotation=<rot 4.5707 90 -90>}]
 thigh_l [
  {shapeType=Capsule, length=25, radius=8, shapeName=thigh_l_capsule, center=<-11.8161 0.0879 0.1447>, rotation=<rot 0 -89.0004 -92.1316>}
  {
   shapeType=Capsule
   length=10.5
   radius=5.5
   shapeName=thigh_l_capsule
   center=<-3.4615 -5.667 -0.2225>
   rotation=<rot 87.8977 179.9984 175.4113>
  }
 ]
 thigh_r [
  {shapeType=Capsule, length=25, radius=7.3, shapeName=thigh_r_capsule, center=<11.8106 -1.0161 0.4401>, rotation=<rot 0 -89.0004 -92.1316>}
  {shapeType=Capsule, length=10.5, radius=5.5, shapeName=thigh_r_capsule, center=<4.3292 5.9052 0>, rotation=<rot 90 6.8146 0>}
 ]
}
constraints [
 {bone1Name=dyn_jacket_bk_02_l, bone2Name=dyn_jacket_bk_01_l, swing1LimitDegrees=10, swing2LimitDegrees=20, twistLimitDegrees=0}
 {bone1Name=dyn_jacket_bk_02, bone2Name=dyn_jacket_bk_01, swing1LimitDegrees=10, swing2LimitDegrees=20, twistLimitDegrees=0}
 {bone1Name=dyn_jacket_bk_02_r, bone2Name=dyn_jacket_bk_01_r, swing1LimitDegrees=10, swing2LimitDegrees=20, twistLimitDegrees=0}
 {bone1Name=dyn_jacket_fr_01_l, bone2Name=dyn_jacket_fr_base_l, swing1LimitDegrees=20, swing2LimitDegrees=50, twistLimitDegrees=10}
 {bone1Name=dyn_jacket_fr_02_l, bone2Name=dyn_jacket_fr_01_l, swing1LimitDegrees=20, swing2LimitDegrees=50, twistLimitDegrees=15}
 {bone1Name=dyn_jacket_fr_01_r, bone2Name=dyn_jacket_fr_base_r, swing1LimitDegrees=10, swing2LimitDegrees=40, twistLimitDegrees=5}
 {bone1Name=dyn_jacket_fr_02_r, bone2Name=dyn_jacket_fr_01_r, swing1LimitDegrees=10, swing2LimitDegrees=40, twistLimitDegrees=5}
 {bone1Name=dyn_jacket_side_01_l, bone2Name=dyn_jacket_side_base_l, swing1LimitDegrees=10, swing2LimitDegrees=20, twistLimitDegrees=5}
 {bone1Name=dyn_jacket_side_02_l, bone2Name=dyn_jacket_side_01_l, swing1LimitDegrees=10, swing2LimitDegrees=20, twistLimitDegrees=5}
 {bone1Name=dyn_jacket_side_01_r, bone2Name=dyn_jacket_side_base_r, swing1LimitDegrees=10, swing2LimitDegrees=20, twistLimitDegrees=0}
 {bone1Name=dyn_jacket_side_02_r, bone2Name=dyn_jacket_side_01_r, swing1LimitDegrees=20, swing2LimitDegrees=20, twistLimitDegrees=5}
 {
  bone1Name=dyn_jacket_fr_base_l
  bone2Name=pelvis
  swing1LimitDegrees=45
  swing1Motion=Free
  swing2LimitDegrees=45
  swing2Motion=Free
  twistLimitDegrees=45
  twistMotion=Free
 }
 {
  bone1Name=dyn_jacket_side_base_l
  bone2Name=pelvis
  swing1LimitDegrees=45
  swing1Motion=Free
  swing2LimitDegrees=45
  swing2Motion=Free
  twistLimitDegrees=45
  twistMotion=Free
 }
 {
  bone1Name=dyn_jacket_bk_base_l
  bone2Name=pelvis
  swing1LimitDegrees=45
  swing1Motion=Free
  swing2LimitDegrees=45
  swing2Motion=Free
  twistLimitDegrees=45
  twistMotion=Free
 }
 {
  bone1Name=dyn_jacket_bk_base
  bone2Name=pelvis
  swing1LimitDegrees=45
  swing1Motion=Free
  swing2LimitDegrees=45
  swing2Motion=Free
  twistLimitDegrees=45
  twistMotion=Free
 }
 {
  bone1Name=dyn_jacket_bk_base_r
  bone2Name=pelvis
  swing1LimitDegrees=45
  swing1Motion=Free
  swing2LimitDegrees=45
  swing2Motion=Free
  twistLimitDegrees=45
  twistMotion=Free
 }
 {
  bone1Name=dyn_jacket_side_base_r
  bone2Name=pelvis
  swing1LimitDegrees=45
  swing1Motion=Free
  swing2LimitDegrees=45
  swing2Motion=Free
  twistLimitDegrees=45
  twistMotion=Free
 }
 {
  bone1Name=dyn_jacket_fr_base_r
  bone2Name=pelvis
  swing1LimitDegrees=45
  swing1Motion=Free
  swing2LimitDegrees=45
  swing2Motion=Free
  twistLimitDegrees=45
  twistMotion=Free
 }
 {
  bone1Name=spine_01
  bone2Name=pelvis
  swing1LimitDegrees=45
  swing1Motion=Free
  swing2LimitDegrees=45
  swing2Motion=Free
  twistLimitDegrees=45
  twistMotion=Free
 }
 {
  bone1Name=thigh_l
  bone2Name=pelvis
  swing1LimitDegrees=45
  swing1Motion=Free
  swing2LimitDegrees=45
  swing2Motion=Free
  twistLimitDegrees=45
  twistMotion=Free
 }
 {
  bone1Name=thigh_r
  bone2Name=pelvis
  swing1LimitDegrees=45
  swing1Motion=Free
  swing2LimitDegrees=45
  swing2Motion=Free
  twistLimitDegrees=45
  twistMotion=Free
 }
 {bone1Name=dyn_jacket_bk_01, bone2Name=dyn_jacket_bk_base, swing1LimitDegrees=10, swing2LimitDegrees=10, twistLimitDegrees=5}
 {bone1Name=dyn_jacket_bk_01_l, bone2Name=dyn_jacket_bk_base_l, swing1LimitDegrees=10, swing2LimitDegrees=10, twistLimitDegrees=5}
 {bone1Name=dyn_jacket_bk_01_r, bone2Name=dyn_jacket_bk_base_r, swing1LimitDegrees=10, swing2LimitDegrees=10, twistLimitDegrees=5}
 {
  bone1Name=dyn_jacket_fr_02_l
  bone2Name=dyn_jacket_side_02_l
  swing1LimitDegrees=45
  swing1Motion=Free
  swing2LimitDegrees=45
  swing2Motion=Free
  twistLimitDegrees=45
  twistMotion=Free
 }
]
