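# Post-style cosmetic AnimBP for the BoldDormRise male medium body, adding secondary motion to a pose copied from the attached parent mesh.
# A Control Rig adjusts that pose, then a weighted constraint drives the medallion base from spine, clavicle and pec bones.
# Two RigidBody passes simulate body and medallion physics assets, followed by Anim Dynamics on jacket slider and pockets.
# Each update sets a gravity override variable from a parameter asset, though simulation nodes show fixed gravity.
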AnimBlueprint /BRCosmetics/Characters/Player/Male/Medium/Bodies/M_MED_BoldDormRise/Meshes/M_MED_BoldDormRise_AnimBP
target_skeleton=&/M_MED_BoldDormRise_Skeleton
graphs [
 AnimGraph {nodes [
   n1 = Control Rig(Source=Copy Pose From Mesh() {bUseAttachedParent=true}) {controlRigAssetReference {blueprintRigClass=&/M_MED_BoldDormRise_CtrlRig.M_MED_BoldDormRise_CtrlRig_C}}
   n2 = "Constraint - dyn_medallion_base (PPPPPPPP)"!(
    ComponentPose=Local To Component(LocalPose=@n1)
    Alpha=1
    ConstraintWeights_0=0.019
    ConstraintWeights_1:float=0.0718
    ConstraintWeights_2:float=0.2745
    ConstraintWeights_3:float=0.5824
    ConstraintWeights_4:float=0.0254
    ConstraintWeights_5:float=0.0269
    ConstraintWeights_6:float=0
    ConstraintWeights_7:float=0
   ) -> (Pose) {
    boneToModify {boneName=dyn_medallion_base}
    constraintSetup [
     {targetBone {boneName=spine_02}}
     {targetBone {boneName=spine_03}}
     {targetBone {boneName=spine_04}}
     {targetBone {boneName=spine_05}}
     {targetBone {boneName=clavicle_l}}
     {targetBone {boneName=clavicle_r}}
     {targetBone {boneName=deform_pec_r}}
     {targetBone {boneName=deform_pec_l}}
    ]
    constraintWeights [1,1,1,1,1,1,1,1]
   }
   n3 = RigidBody!(ComponentPose=@n2, Alpha, OverrideWorldGravity:Vector="0,0,-980", ExternalForce) -> (Pose) {
    bClampLinearTranslationLimitToRefPose=true
    bOverrideWorldGravity=true
    overridePhysicsAsset=&/M_MED_BoldDormRise_Physics
    simulationSpace=BaseBoneSpace
    baseBoneRef {boneName=pelvis}
    overrideWorldGravity {z=-980}
    simSpaceSettings {
     dampingAlpha=0.1
     maxAngularAcceleration=400
     maxAngularVelocity=10
     maxLinearAcceleration=2000
     maxLinearVelocity=600
     velocityScaleZ=0.8
     worldAlpha=0.6
    }
   }
   n4 = RigidBody(ComponentPose=@n3, OverrideWorldGravity="0,0,-980") {
    bClampLinearTranslationLimitToRefPose=true
    bOverrideWorldGravity=true
    overridePhysicsAsset=&/M_MED_BoldDormRise_Medallion_Physics
    simulationSpace=BaseBoneSpace
    baseBoneRef {boneName=spine_05}
    overrideWorldGravity {z=-980}
    simSpaceSettings {dampingAlpha=0, maxAngularAcceleration=300, maxAngularVelocity=8, maxLinearAcceleration=3500, maxLinearVelocity=700, worldAlpha=0.8}
   }
   n5 = Anim Dynamics - Bone: dyn_jacket_slider!(ComponentPose=@n4, Alpha=1, GravityOverride:Vector="0,0,-980", ExternalForce="0,0,0") -> (Pose) {
    angularDampingOverride=0.999
    bOverrideAngularDamping=true
    bUseGravityOverride=true
    simulationSpace=RootRelative
    boundBone {boneName=dyn_jacket_slider}
    gravityOverride {z=-980}
    physicsBodyDefinitions [{
      boundBone {boneName=dyn_jacket_slider}
      boxExtents=<2 2 2>
      constraintSetup {angularLimitsMax=<10 30 15>, angularLimitsMin=<-10 0 -15>, angularTarget=<0 0 0>}
      localJointOffset=<0.2 0 0>
     }]
   }
   n6 = Anim Dynamics - Bone: dyn_pocket_l!(ComponentPose=@n5, Alpha=1, GravityOverride:Vector="0,0,-980", ExternalForce="0,0,0") -> (Pose) {
    angularDampingOverride=0.999
    bOverrideAngularDamping=true
    bUseGravityOverride=true
    simulationSpace=RootRelative
    boundBone {boneName=dyn_pocket_l}
    gravityOverride {z=-980}
    physicsBodyDefinitions [{
      boundBone {boneName=dyn_pocket_l}
      boxExtents=<2 2 2>
      constraintSetup {angularLimitsMax=<0 40 0>, angularLimitsMin=<0 0 0>, angularTarget=<0 0 0>}
      localJointOffset=<0.2 0 0>
     }]
   }
   n7 = Anim Dynamics - Bone: dyn_pocket_r!(ComponentPose=@n6, Alpha=1, GravityOverride:Vector="0,0,-980", ExternalForce="0,0,0") -> (Pose) {
    angularDampingOverride=0.999
    bOverrideAngularDamping=true
    bUseGravityOverride=true
    simulationSpace=RootRelative
    boundBone {boneName=dyn_pocket_r}
    gravityOverride {z=-980}
    physicsBodyDefinitions [{
      boundBone {boneName=dyn_pocket_r}
      boxExtents=<2 2 2>
      constraintSetup {angularLimitsMax=<0 40 0>, angularLimitsMin=<0 0 0>, angularTarget=<0 0 0>}
      localJointOffset=<0.2 0 0>
     }]
   }
   Output Pose(Result=Component To Local(ComponentPose=@n7)) {layerGroup=None}
  ]}
 EventGraph {nodes [
   Event BlueprintUpdateAnimation()
   > n1 = "Gravity Override State Function (Parameter Asset) (Nativized)"(
    CharacterPartAnimInstance=Self-Reference()
    GameTimeInSeconds=GetGameTimeInSeconds()
    Parameters=&/M_MED_BoldDormRise_Parameters
   )
   > Set Gravity Override(Gravity Override=@n1)
  ]}
]
variables [
 Gravity Override {type=Vector}
 Gravity Override Medalion {type=Vector}
 Gravity Override Params {type=GravityOverrideParamsStruct}
 Gravity Override Params Medalion {type=GravityOverrideParamsStruct, default {dBNO {gravityOverride=<-100 400 0>}}}
]
